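# Editor utility widget that serves as a combo box entry for map marker icon overrides, located in the quest editor tools.
# It lays out an inherited lazy-loaded icon image beside a small, wrapping name label in a horizontal box, with the label filling the remaining width.
# The Blueprint supplies only layout and styling; its event stubs are empty, leaving any behavior to the native FortMapMarkerIconOverrideComboBoxEntry parent.
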EditorUtilityWidgetBlueprint /Game/NeverCook/EditorUtilities/QuestTools/QuestEditor/QuestItemDefComponentUI/EditorUtilityWidgets/MapIconOverrideComboBoxEntry
graphs [EventGraph {nodes [
   Event PreConstruct()
   Event Construct()
   Event Tick()
  ]}]
variables []
widgets {
 info {
  inheritedWidgetCount=2
  parentClass=/Script/FortniteEditor.FortMapMarkerIconOverrideComboBoxEntry
  rootWidgetClass=&HorizontalBox
  widgetCount=3
 }
 widgets [
  HorizontalBox_93 {widgetClassPath=&HorizontalBox, widget=HorizontalBox_93}
  Image_Icon {
   widgetClassPath=&FortLazyImage
   bInherited=true
   bIsVariable=true
   parent=HorizontalBox_93
   widget=Image_Icon
   slot {class=HorizontalBoxSlot, properties {padding=<ltrb 5 0 5 0>}}
  }
  Text_IconName {
   widgetClassPath=&CommonTextBlock
   bInherited=true
   parent=HorizontalBox_93
   widget=Text_IconName
   properties {
    autoWrapText=true
    text=Text Block
    font {
     bForceMonospaced=false
     bMaterialIsStencil=false
     fontMaterial=null
     fontObject=&NotoSans
     letterSpacing=0
     monospacedWidth=1
     size=8
     skewAmount=0
     typefaceFontName=NotoSans-Regular
     outlineSettings {
      bApplyOutlineToDropShadows=false
      bMiteredCorners=false
      bSeparateFillAlpha=false
      outlineMaterial=null
      outlineSize=0
      outlineColor=<rgba 0 0 0 1>
     }
    }
    shadowOffset=<0 0>
   }
   slot {class=HorizontalBoxSlot, properties {verticalAlignment=VAlign_Center, size {sizeRule=Fill, value=1}}}
  }
 ]
}
